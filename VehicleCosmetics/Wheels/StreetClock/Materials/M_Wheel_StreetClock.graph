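# Vehicle wheel material for StreetClock that animates a short and a long clock arrow on top of MF_WheelParent_Simple.
# Two time-driven RotateAboutAxis offsets spin around the local Y axis at the object pivot, with vertex color blue and green weighting each arrow.
# Their sum, scaled by the Animation parameter, replaces the parent's World Position Offset.
# The result then passes through the vehicle damage and de-rez transition functions.
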Material /VehicleCosmetics/Wheels/StreetClock/Materials/M_Wheel_StreetClock
graph {
 properties {
  bAutomaticallySetUsageInEditor=false
  bUseMaterialAttributes=true
  bUsedWithMeshDeformer=true
  bUsedWithSkeletalMesh=true
  pixelDepthOffsetMode=PDOM_Legacy
  previewMesh=/VehicleCosmetics/Wheels/StreetClock/Meshes/SK_Wheel_StreetClock
  usedShadingModels=Default Lit
 }
 nodes [
  Constant()
  n1 = GetMaterialAttributes(None=MaterialFunctionCall() {materialFunction=&MF_WheelParent_Simple}.MaterialAttributes) {attributeGetTypes [#1, #2]}
  n2 = Constant3Vector() {constant=<rgba 0 1 0 0>}
  n3 = Normalize(VectorInput=Multiply(A=Transform(None=@n2) {transformSourceType=TRANSFORMSOURCE_Local}, B=ObjectRadius()))
  n4 = Multiply(A=Time(), B=Constant() {r=1})
  n5 = ScalarParameter() {defaultValue=12, parameterName=SpeedShortArrow}
  n6 = VertexColor()
  n7 = Constant()
  n8 = RotateAboutAxis(
   NormalizedRotationAxis=@n3
   RotationAngle=LinearInterpolate(A=Constant(), B=Divide(A=@n4, B=@n5), Alpha=LinearInterpolate(A=@n6.B, B=@n7, Alpha=@n6.B))
   PivotPoint=MaterialFunctionCall() {materialFunction=/VehicleCosmetics/Wheels/Visceral_Tier1/Materials/ObjectPivotPoint}.ObjectPivotLoacation
   Position=WorldPosition().XYZ
  )
  n9 = Multiply(A=Time(), B=Constant() {r=1})
  n10 = ScalarParameter() {defaultValue=6, parameterName=SpeedLongArrow}
  n11 = VertexColor()
  n12 = Constant()
  n13 = RotateAboutAxis(
   NormalizedRotationAxis=@n3
   RotationAngle=LinearInterpolate(A=Constant(), B=Divide(A=@n9, B=@n10), Alpha=LinearInterpolate(A=@n11.G, B=@n12, Alpha=@n11.G))
   PivotPoint=MaterialFunctionCall() {materialFunction=&ObjectPivotPoint}.ObjectPivotLoacation
   Position=WorldPosition().XYZ
  )
  n14 = LinearInterpolate(
   A=@n1.World Position Offset
   B=Multiply(A=Add(A=@n8, B=@n13), B=ScalarParameter() {defaultValue=1, parameterName=Animation})
   Alpha=Constant() {r=1}
  )
  MaterialFunctionCall(In=MaterialFunctionCall(Material=SetMaterialAttributes!(MaterialAttributes=@n1.MaterialAttributes, Emissive Color=@n1.MaterialAttributes, World Position Offset=@n14) {attributeSetTypes [#1, #2]}) {materialFunction=&MF_BaseVehicle_Damage}.Result) {materialFunction=&MF_VehicleDeRez_Transition}
 ]
}
