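# Translucent material for the StreetClock vehicle wheel logo, showing a two-color sticker on tinted glass.
# A rotated mask texture uses G to separate sticker from background for color and opacity, and A to blend text and logo colors.
# The background combines sharpened vertex-color blue with Fresnel.
# Base color and opacity pass through vehicle damage and de-rez functions, while the roughness and metallic blends are left unconnected.
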Material /VehicleCosmetics/Wheels/StreetClock/Materials/M_Wheel_StreetClock_Logo
graph {
 properties {
  bAutomaticallySetUsageInEditor=false
  bUseMaterialAttributes=true
  bUsedWithSkeletalMesh=true
  blendMode=BLEND_Translucent
  usedShadingModels=Default Lit
 }
 nodes [
  Constant() {r=0.36}
  n1 = TextureSample(UVs=MaterialFunctionCall(UVs=TextureCoordinate(), Rotation Center=Constant2Vector() {g=0.5, r=0.5}, "Rotation Angle (0-1)"=Constant() {r=-0.1}) {materialFunction=&CustomRotator}.Rotated Values) {
   samplerSource=SSM_Wrap_WorldGroupSettings
   samplerType=SAMPLERTYPE_LinearColor
   texture=/VehicleCosmetics/Wheels/StreetClock/Textures/T_Wheel_StreetClock_Logo_M
  }
  LinearInterpolate(
   A=ScalarParameter() {defaultValue=0.2, parameterName=Roughness_Background}
   B=ScalarParameter() {defaultValue=1, parameterName=Roughness_Sticker}
   Alpha=@n1.G
  )
  LinearInterpolate(A=ScalarParameter() {defaultValue=1, parameterName=Metallic_Background}, B=ScalarParameter() {parameterName=Metallic_Sticker}, Alpha=@n1.G)
  n2 = Clamp(None=Power(Base=VertexColor().B, Exp=Constant() {r=32}))
  n3 = MaterialFunctionCall!(Normal, Exponent) {materialFunction=/VehicleCosmetics/Wheels/Cola_Holographic/Materials/MF_Fresnel_Classic}
  n4 = VectorParameter() {parameterName=CustomColor_Text, defaultValue=<rgba 0 0 0 1>}
  n5 = VectorParameter() {parameterName=CustomColor_Logo, defaultValue=<rgba 1 0 0.0426 1>}
  n6 = Constant()
  n7 = Constant() {r=0.95}
  n8 = Multiply(
   A=VectorParameter() {parameterName=BackGround_TInt, defaultValue=<rgba 0.6666 0.7718 0.9358 1>}.RGB
   B=LinearInterpolate(A=Multiply(A=@n2, B=@n3.Result), B=LinearInterpolate(A=@n4.RGB, B=@n5.RGB, Alpha=@n1.A), Alpha=Clamp(None=@n1.G, Min=@n6, Max=@n7))
  )
  n9 = LinearInterpolate(
   A=ScalarParameter() {defaultValue=0.5, parameterName=Glass_Opacity}
   B=ScalarParameter() {defaultValue=1, parameterName=Sticker_Opacity}
   Alpha=@n1.G
  )
  MaterialFunctionCall(In=MaterialFunctionCall(Material=SetMaterialAttributes!(MaterialAttributes, Base Color=@n8, Opacity=@n9) {attributeSetTypes [#1, #2]}) {materialFunction=&MF_BaseVehicle_Damage}.Result) {materialFunction=&MF_VehicleDeRez_Transition}
 ]
}
